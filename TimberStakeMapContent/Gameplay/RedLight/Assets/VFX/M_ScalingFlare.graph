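# Additive unlit flare material for Niagara sprite and mesh particles, built to enlarge flares as the camera gets farther away.
# Camera-to-particle distance past a tunable offset, normalized and clamped, blends in a vertex-offset vector scaled from the particle center up to a maximum increase.
# Color comes from blending an outer flare mask with a spiky flare mask via an anamorphic parameter, tinted by particle color.
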Material /TimberStakeMapContent/Gameplay/RedLight/Assets/VFX/M_ScalingFlare
graph {
 properties {
  bAutomaticallySetUsageInEditor=false
  bCastRayTracedShadows=false
  bUsedWithNiagaraMeshParticles=true
  bUsedWithNiagaraSprites=true
  blendMode=BLEND_Additive
  shadingModel=MSM_Unlit
  usedShadingModels=Unlit
 }
 nodes [
  DepthFade() {fadeDistanceDefault=8}
  n1 = ParticlePositionWS()
  n2 = Subtract(A=WorldPosition().XYZ, B=@n1)
  n3 = Subtract(A=Distance(A=@n1, B=CameraPositionWS()), B=ScalarParameter() {defaultValue=2000, parameterName=DistanceOffset})
  n4 = ScalarParameter() {defaultValue=30000, parameterName=ScaleUpDistance}
  LinearInterpolate(B=Multiply(A=ScalarParameter() {defaultValue=10, parameterName=MaxScaleIncrease}, B=@n2) {constB=-1}, Alpha=Clamp(None=Divide(A=@n3, B=@n4))) {constB=0}
  Multiply(A=@n2) {constB=-1}
  n5 = TextureCoordinate()
  n6 = TextureSampleParameter2D(UVs=@n5) {parameterName=OuterFlareMaskTexture, samplerType=SAMPLERTYPE_Grayscale, texture=/Game/Effects/Fort_Effects/Textures/T_EFX_CameraFlare_01}
  n7 = TextureSampleParameter2D(UVs=@n5) {parameterName=FlareMask, texture=/Game/Effects/Fort_Effects/Textures/Flares/T_Flare_Spikeball}
  n8 = ScalarParameter() {defaultValue=1, parameterName=Lerp Anamorphic, sliderMax=1}
  Reroute(None=Multiply(A=LinearInterpolate(A=@n6.RGB, B=@n7.RGB, Alpha=@n8) {constAlpha=0.2}, B=ParticleColor().RGB))
 ]
}
