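# Niagara system for the glowing eyes in the Red Light gameplay, tinted by a system color and sized by a global scale.
# One emitter bursts two particles, mirrors them into eye positions by exec index, and draws each with two sprite renderers and a mesh renderer.
# A second emitter drives an offset light with user-set intensity, radius and offset, multiplying color by one of its own channels, apparently green to limit it to green-light states.
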NiagaraSystem /TimberStakeMapContent/Gameplay/RedLight/Assets/VFX/NS_RedlLight_EyeGlows
dependencies {
 usedDataInterfaces []
 usedDynamicInputs [&MakeFloatFromVector]
 usedModules [
  &SystemState
  &self:SystemSpawnScript.SystemScriptSource.SystemScriptGraph.NiagaraNodeAssignment_0.SetVariables_#1
  &self:SystemSpawnScript.SystemScriptSource.SystemScriptGraph.NiagaraNodeAssignment_1.SetVariables_#2
 ]
}
emitters {
 SingleLoopingParticle {
  emitterUpdateScript [
   &EmitterState()
   &SpawnBurst_Instantaneous(Spawn Count=2, Use Spawn Probability=false, UseLoopCountLimit=false)
  ]
  particleSpawnScript [&InitializeParticle(
    Color=$System.Color
    Mesh Scale Mode=::NewEnumerator3
    Mesh Uniform Scale=5
    Sprite Size=&Multiply_Vector2D_ByFloat(Float=$System.GlobalScale, Vector2D=<12 12>)
    Sprite Size Mode=::NewEnumerator4
    UsePositionOffset=false
    Write Material Random=true
   )]
  particleUpdateScript [
   &ParticleState()
   /SetVariables(Particles.Color=&Multiply_LinearColorByFloat(Linear Color=$System.Color, Float=12)) {isSetParameters=true}
   /SetVariables(Particles.Color_B:LinearColor=&Multiply_LinearColorByFloat(Linear Color=$System.Color, Float=5)) {isSetParameters=true}
   /SetVariables(Particles.Position=&AddVectorToPosition(
     Position=<0 0 0>
     Vector=&MaskVector(
      Mask Source=1
      Mask Parameter=&ReturnExecIndex()
      Mask=0
      Masked Vector=<-18 33 24.5>
      Passthrough Vector=<18 33 24.5>
     )
    )) {isSetParameters=true, module=SetVariables_#3}
   &CameraOffset(Camera Offset Amount=&Multiply_Float(A=5, B=$System.GlobalScale))
   /SetVariables(Particles.SpriteSize_B:Vector2f=&Multiply_Vector2D_ByFloat(Float=$System.GlobalScale, Vector2D=<25 25>)) {isSetParameters=true}
   /SetVariables(Particles.CamOffset_Far:NiagaraFloat=&Multiply_Float(A=100, B=$System.GlobalScale)) {isSetParameters=true}
  ]
  rendererClasses [&NiagaraSpriteRendererProperties, &NiagaraMeshRendererProperties]
  renderers [
   {rendererClass=&NiagaraSpriteRendererProperties}
   {rendererClass=&NiagaraSpriteRendererProperties, rendererIndex=1}
   {rendererClass=&NiagaraMeshRendererProperties, rendererIndex=2}
  ]
 }
 SingleLoopingParticle001 {
  emitterUpdateScript [
   &EmitterState()
   &SpawnBurst_Instantaneous(Use Spawn Probability=false, UseLoopCountLimit=false)
  ]
  particleSpawnScript [&InitializeParticle(
    Color=$System.Color
    Mesh Scale Mode=::NewEnumerator3
    Sprite Size Mode=::NewEnumerator3
    Uniform Sprite Size=15
    UsePositionOffset=false
    Write Material Random=true
   )]
  particleUpdateScript [
   &ParticleState()
   /SetVariables(Particles.Color=&Multiply_LinearColorByFloat(Linear Color=$System.Color, Float=$User.Green Light Intensity)) {isSetParameters=true}
   /SetVariables(Particles.Position=&AddVectorToPosition(Position=<0 0 0>, Vector=&MakeVector(X=0, Y=$User.Green Light Offset, Z=24.5))) {isSetParameters=true, module=SetVariables_#3}
   /SetVariables(Particles.Color=&Multiply_LinearColorByFloat(
     Linear Color=$Particles.Color
     Float=&MakeFloatFromLinearColor(LinearColor=$Particles.Color, Channel:ENiagaraFloatFromLinearColorOptions=::NewEnumerator1)
    )) {isSetParameters=true}
   /SetVariables(Particles.LightRadius=&Multiply_Float(A=$User.Green Light Radius, B=$System.GlobalScale)) {isSetParameters=true}
  ]
  rendererClasses [&NiagaraLightRendererProperties]
  renderers [{rendererClass=&NiagaraLightRendererProperties, rendererIndex=0}]
 }
}
summary {
 systemName=NS_RedlLight_EyeGlows
 emitters [
  {
   emitterName=SingleLoopingParticle
   rendererClasses [&NiagaraSpriteRendererProperties, &NiagaraMeshRendererProperties]
  }
  {emitterName=SingleLoopingParticle001, rendererClasses [&NiagaraLightRendererProperties]}
 ]
 userVariables [
  User.Color {type=LinearColor, description="", defaultValue {struct=LinearColor, value=<rgba 1 0 0 1>}}
  User.Green Light Intensity {type=NiagaraFloat, defaultValue=1, description=""}
  User.Green Light Offset {type=NiagaraFloat, defaultValue=125, description=""}
  User.Green Light Radius {type=NiagaraFloat, defaultValue=200, description=""}
 ]
}
